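# Material function for the Athena landscape that exposes terrain biome masks for Mountain, Farm and Swamp.
# It samples the overridable BiomeMask texture parameter in the TerrainColorMatch group, using UVs from the shared terrain coordinates function's WorldPosition call.
# The red, green and blue channels of that single linear-color texture become the Farm, Mountain and Swamp outputs respectively.
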MaterialFunction /Clyde/Athena/Environments/Landscape/MaterialFunctions/MF_Athena_BiomeMask
graph {nodes [
  n1 = TextureSampleParameter2D(UVs=MaterialFunctionCall!(WorldPosition) {materialFunction=/Game/Athena/Environments/Landscape/MaterialFunctions/Utilities/MF_Athena_Terrain_Coords}.Result) {
   constMipValue=5
   group=TerrainColorMatch
   parameterName=BiomeMask
   samplerType=SAMPLERTYPE_LinearColor
   texture=/Game/Athena/Environments/Landscape/Textures/T_Athena_Terrain_Biome_M
  }
  FunctionOutput(None=@n1.G) {outputName=Mountain}
  FunctionOutput(None=@n1.R) {outputName=Farm}
  FunctionOutput(None=@n1.B) {outputName=Swamp}
 ]}
